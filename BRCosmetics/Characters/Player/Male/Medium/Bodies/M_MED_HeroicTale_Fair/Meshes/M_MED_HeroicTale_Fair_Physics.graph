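# Physics asset for the M_MED_HeroicTale_Fair male player body, covering the front and back loincloth flaps.
# Capsules on the pelvis and both upper thigh twist bones form the body shapes, alongside tiny capsules on the four loincloth base bones.
# Base and thigh bodies join the pelvis with free angular motion.
# Box bodies on the dyn bones hang from the bases with angular limits: one segment per back flap, two chained per front flap.
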PhysicsAsset /BRCosmetics/Characters/Player/Male/Medium/Bodies/M_MED_HeroicTale_Fair/Meshes/M_MED_HeroicTale_Fair_Physics
bodies {
 base_backLoinCloth_1_l [{shapeType=Capsule, length=0.5, radius=0.25, shapeName=base_backLoinCloth_1_l_capsule}]
 base_backLoinCloth_1_r [{shapeType=Capsule, length=0.5, radius=0.25, shapeName=base_backLoinCloth_1_r_capsule}]
 base_frontLoinCloth_1_l [{shapeType=Capsule, length=0.5, radius=0.25, shapeName=base_backLoinCloth_1_r_capsule}]
 base_frontLoinCloth_1_r [{shapeType=Capsule, length=0.5, radius=0.25, shapeName=base_backLoinCloth_1_r_capsule}]
 dyn_backLoinCloth_1_l [{shapeType=Box, extentX=11.7006, extentY=8.9604, extentZ=1.6905, shapeName=dyn_backLoinCloth_1_l_box, center=<5.703 0 0.2118>}]
 dyn_backLoinCloth_1_r [{shapeType=Box, extentX=11.7006, extentY=8.9604, extentZ=1.6905, shapeName=dyn_backLoinCloth_1_r_box, center=<5.703 0 0.2118>}]
 dyn_frontLoinCloth_1_l [{shapeType=Box, extentX=7.7591, extentY=12.198, extentZ=1.6905, shapeName=dyn_backLoinCloth_1_l_box, center=<5.5 0 0.6946>}]
 dyn_frontLoinCloth_1_r [{shapeType=Box, extentX=7.7591, extentY=12.198, extentZ=1.6905, shapeName=dyn_backLoinCloth_1_l_box, center=<5.5 0 0.6946>}]
 dyn_frontLoinCloth_2_l [{shapeType=Box, extentX=7.7591, extentY=11.4982, extentZ=1.6905, shapeName=dyn_backLoinCloth_1_l_box, center=<5.5 0.9615 0.1105>}]
 dyn_frontLoinCloth_2_r [{
   shapeType=Box
   extentX=7.7591
   extentY=11.4982
   extentZ=1.6905
   shapeName=dyn_backLoinCloth_1_r_box
   center=<5.5 -0.9615 0.1105>
   rotation=<rot 0 0 -180>
  }]
 pelvis [{shapeType=Capsule, length=10.3194, radius=10, shapeName=pelvis_capsule, center=<-4.7854 -0.4088 0>}]
 thigh_twist_01_l [{shapeType=Capsule, length=18.7152, radius=9.2306, shapeName=thigh_twist_01_l_capsule, center=<7.523 0 0>, rotation=<rot 90 -2.7707 0>}]
 thigh_twist_01_r [{shapeType=Capsule, length=18.7152, radius=9.2306, shapeName=thigh_twist_01_r_capsule, center=<-7.523 0 0>, rotation=<rot 90 177.2293 0>}]
}
constraints [
 {bone1Name=dyn_backLoinCloth_1_r, bone2Name=base_backLoinCloth_1_r, swing1LimitDegrees=15, swing2LimitDegrees=60, twistLimitDegrees=0}
 {bone1Name=dyn_backLoinCloth_1_l, bone2Name=base_backLoinCloth_1_l, swing1LimitDegrees=15, swing2LimitDegrees=60, twistLimitDegrees=0}
 {
  bone1Name=base_backLoinCloth_1_l
  bone2Name=pelvis
  swing1LimitDegrees=45
  swing1Motion=Free
  swing2LimitDegrees=45
  swing2Motion=Free
  twistLimitDegrees=45
  twistMotion=Free
 }
 {
  bone1Name=base_backLoinCloth_1_r
  bone2Name=pelvis
  swing1LimitDegrees=45
  swing1Motion=Free
  swing2LimitDegrees=45
  swing2Motion=Free
  twistLimitDegrees=45
  twistMotion=Free
 }
 {
  bone1Name=base_frontLoinCloth_1_l
  bone2Name=pelvis
  swing1LimitDegrees=45
  swing1Motion=Free
  swing2LimitDegrees=45
  swing2Motion=Free
  twistLimitDegrees=45
  twistMotion=Free
 }
 {bone1Name=dyn_frontLoinCloth_1_l, bone2Name=base_frontLoinCloth_1_l, swing1LimitDegrees=14, swing2LimitDegrees=30, twistLimitDegrees=10}
 {bone1Name=dyn_frontLoinCloth_2_l, bone2Name=dyn_frontLoinCloth_1_l, swing1LimitDegrees=10, swing2LimitDegrees=40, twistLimitDegrees=5}
 {
  bone1Name=base_frontLoinCloth_1_r
  bone2Name=pelvis
  swing1LimitDegrees=45
  swing1Motion=Free
  swing2LimitDegrees=45
  swing2Motion=Free
  twistLimitDegrees=45
  twistMotion=Free
 }
 {bone1Name=dyn_frontLoinCloth_1_r, bone2Name=base_frontLoinCloth_1_r, swing1LimitDegrees=14, swing2LimitDegrees=30, twistLimitDegrees=10}
 {bone1Name=dyn_frontLoinCloth_2_r, bone2Name=dyn_frontLoinCloth_1_r, swing1LimitDegrees=10, swing2LimitDegrees=40, twistLimitDegrees=5}
 {
  bone1Name=thigh_twist_01_l
  bone2Name=pelvis
  swing1LimitDegrees=45
  swing1Motion=Free
  swing2LimitDegrees=45
  swing2Motion=Free
  twistLimitDegrees=45
  twistMotion=Free
 }
 {
  bone1Name=thigh_twist_01_r
  bone2Name=pelvis
  swing1LimitDegrees=45
  swing1Motion=Free
  swing2LimitDegrees=45
  swing2Motion=Free
  twistLimitDegrees=45
  twistMotion=Free
 }
]
