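# Physics asset for the skirt accessories of the male medium HeroicTale Fair cosmetic body, covering a sword, two hip pockets and a three-link key chain.
# Each accessory hangs from a tiny base capsule tied to spine_01, with tight swing and twist limits on its dyn bones.
# Capsules on pelvis, spine, thighs, right chest pocket and front loincloth add body shapes, joined by free-rotation constraints.
# Shapes are capsules except the key chain's boxed end link.
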PhysicsAsset /BRCosmetics/Characters/Player/Male/Medium/Bodies/M_MED_HeroicTale_Fair/Meshes/M_MED_HeroicTale_Fair_skirtAcc_Physics
bodies {
 base_hipPocket_1_l [{shapeType=Capsule, length=0.25, radius=0.5, shapeName=base_sword_capsule}]
 base_hipPocket_2_l [{shapeType=Capsule, length=0.25, radius=0.5, shapeName=base_sword_capsule}]
 base_keyChain_1 [{shapeType=Capsule, length=0.25, radius=0.5, shapeName=base_sword_capsule}]
 base_sword [{shapeType=Capsule, length=0.25, radius=0.5, shapeName=base_sword_capsule}]
 dyn_chestPocket_outer_r [{
   shapeType=Capsule
   length=8.6645
   radius=4.5602
   shapeName=dyn_chestPocket_outer_r_capsule
   center=<4.5665 1.1007 -1.6857>
   rotation=<rot 0 -88.4252 -87.0273>
  }]
 dyn_frontLoinCloth_2_r [{
   shapeType=Capsule
   length=19.7157
   radius=8.5531
   shapeName=dyn_frontLoinCloth_2_r_capsule
   center=<2.4984 -8.9322 -7.2969>
   rotation=<rot 0 0 90>
  }]
 dyn_hipPocket_1_l [
  {
   shapeType=Capsule
   length=7.7239
   radius=2.3388
   shapeName=dyn_hipPocket_1_l_capsule
   center=<3.4856 1.6256 1.6187>
   rotation=<rot 0 91.5944 -88.6852>
  }
  {
   shapeType=Capsule
   length=7.7239
   radius=2.3388
   shapeName=dyn_hipPocket_1_l_capsule
   center=<3.5851 -2.0686 1.7654>
   rotation=<rot 0 91.5944 -88.6852>
  }
 ]
 dyn_hipPocket_2_l [
  {
   shapeType=Capsule
   length=7.5594
   radius=2.3818
   shapeName=dyn_hipPocket_2_l_capsule
   center=<3.6221 2.1525 1.8181>
   rotation=<rot 0 90.9605 -87.3669>
  }
  {
   shapeType=Capsule
   length=7.5594
   radius=2.3818
   shapeName=dyn_hipPocket_2_l_capsule
   center=<3.6863 -1.666 1.8145>
   rotation=<rot 0 90.9605 -87.3669>
  }
 ]
 dyn_keyChain_1 [{shapeType=Capsule, length=4.1917, radius=1.307, shapeName=dyn_keyChain_1_capsule, center=<2.6599 0 0>, rotation=<rot 0 90 -90>}]
 dyn_keyChain_2 [{shapeType=Capsule, length=5.0229, radius=1.2152, shapeName=dyn_keyChain_2_capsule, center=<4.4559 0 0>, rotation=<rot 0 -90 -90>}]
 dyn_keyChain_3 [{shapeType=Box, extentX=7.5, extentY=6.4994, extentZ=2.6461, shapeName=dyn_keyChain_3_box, center=<5.228 0 0.1511>}]
 dyn_sword [{
   shapeType=Capsule
   length=46.4237
   radius=3.4761
   shapeName=dyn_sword_capsule
   center=<3.0062 -0.0796 -3.046>
   rotation=<rot 0 90.7497 -89.6505>
  }]
 pelvis [{shapeType=Capsule, length=10.2304, radius=10.089, shapeName=pelvis_capsule, center=<-4.7854 -0.4088 0>}]
 spine_01 [{shapeType=Capsule, length=6.7846, radius=9.7383, shapeName=pelvis_capsule, center=<-4.7854 -0.4088 0>}]
 spine_03 [{shapeType=Capsule, length=8.8365, radius=9.0884, center=<0 1.1878 0>}]
 thigh_twist_01_l [{shapeType=Capsule, length=20.2719, radius=7.6739, shapeName=thigh_twist_01_l_capsule, center=<7.523 0 0>, rotation=<rot 90 -1.3445 0>}]
 thigh_twist_01_r [{
   shapeType=Capsule
   length=20.2719
   radius=7.6739
   shapeName=thigh_twist_01_r_capsule
   center=<-7.523 0 -0.6142>
   rotation=<rot 90 -176.5512 0>
  }]
}
constraints [
 {
  bone1Name=thigh_twist_01_l
  bone2Name=pelvis
  swing1LimitDegrees=45
  swing1Motion=Free
  swing2LimitDegrees=45
  swing2Motion=Free
  twistLimitDegrees=45
  twistMotion=Free
 }
 {
  bone1Name=thigh_twist_01_r
  bone2Name=pelvis
  swing1LimitDegrees=45
  swing1Motion=Free
  swing2LimitDegrees=45
  swing2Motion=Free
  twistLimitDegrees=45
  twistMotion=Free
 }
 {bone1Name=dyn_hipPocket_1_l, bone2Name=base_hipPocket_1_l, swing1LimitDegrees=15, swing2LimitDegrees=20, twistLimitDegrees=0}
 {
  bone1Name=base_hipPocket_1_l
  bone2Name=spine_01
  swing1LimitDegrees=45
  swing1Motion=Free
  swing2LimitDegrees=45
  swing2Motion=Free
  twistLimitDegrees=45
  twistMotion=Free
 }
 {bone1Name=dyn_hipPocket_2_l, bone2Name=base_hipPocket_2_l, swing1LimitDegrees=15, swing2LimitDegrees=20, twistLimitDegrees=0}
 {
  bone1Name=base_hipPocket_2_l
  bone2Name=spine_01
  swing1LimitDegrees=45
  swing1Motion=Free
  swing2LimitDegrees=45
  swing2Motion=Free
  twistLimitDegrees=45
  twistMotion=Free
 }
 {
  bone1Name=base_keyChain_1
  bone2Name=spine_01
  swing1LimitDegrees=45
  swing1Motion=Free
  swing2LimitDegrees=45
  swing2Motion=Free
  twistLimitDegrees=45
  twistMotion=Free
 }
 {bone1Name=dyn_keyChain_1, bone2Name=base_keyChain_1, swing1LimitDegrees=20, swing2LimitDegrees=30, twistLimitDegrees=0}
 {bone1Name=dyn_keyChain_2, bone2Name=dyn_keyChain_1, swing1LimitDegrees=20, swing2LimitDegrees=30, twistLimitDegrees=0}
 {bone1Name=dyn_keyChain_3, bone2Name=dyn_keyChain_2, swing1LimitDegrees=20, swing2LimitDegrees=30, twistLimitDegrees=10}
 {bone1Name=dyn_sword, bone2Name=base_sword, swing1LimitDegrees=20, swing2LimitDegrees=20, twistLimitDegrees=10}
 {
  bone1Name=base_sword
  bone2Name=spine_01
  swing1LimitDegrees=45
  swing1Motion=Free
  swing2LimitDegrees=45
  swing2Motion=Free
  twistLimitDegrees=45
  twistMotion=Free
 }
 {
  bone1Name=spine_01
  bone2Name=pelvis
  swing1LimitDegrees=45
  swing1Motion=Free
  swing2LimitDegrees=45
  swing2Motion=Free
  twistLimitDegrees=45
  twistMotion=Free
 }
 {
  bone1Name=dyn_frontLoinCloth_2_r
  bone2Name=pelvis
  swing1LimitDegrees=45
  swing1Motion=Free
  swing2LimitDegrees=45
  swing2Motion=Free
  twistLimitDegrees=45
  twistMotion=Free
 }
 {
  bone1Name=dyn_chestPocket_outer_r
  bone2Name=spine_03
  swing1LimitDegrees=45
  swing1Motion=Free
  swing2LimitDegrees=45
  swing2Motion=Free
  twistLimitDegrees=45
  twistMotion=Free
 }
 {
  bone1Name=spine_03
  bone2Name=spine_01
  swing1LimitDegrees=45
  swing1Motion=Free
  swing2LimitDegrees=45
  swing2Motion=Free
  twistLimitDegrees=45
  twistMotion=Free
 }
]
