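# Animation Blueprint for the HeroicTale Fair male body cosmetic, adding secondary motion to the pose copied from its parent mesh.
# Cached stages run Control Rig, cape, backpack and loincloth offsets, pocket AnimDynamics, two RigidBody simulations, then post-simulation loincloth and belt fixes.
# The event graph derives gravity override and adjustment values from movement, glider and hood state.
# A front-end idle curve blends DefaultSlot onto the sword-handle bone branch.
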AnimBlueprint /BRCosmetics/Characters/Player/Male/Medium/Bodies/M_MED_HeroicTale_Fair/Meshes/M_MED_HeroicTale_Fair_AnimBP
target_skeleton=&/M_MED_HeroicTale_Fair_Skeleton
graphs [
 AnimGraph {nodes [
   Output Pose(Result=Layered blend per bone(
     BasePose=Use cached pose 'postSimFixes'()
     BlendPoses_0=Slot 'DefaultSlot'()
     BlendWeights_0=Get Custom Idle is Playing Float() -> (Custom Idle is Playing Float:double)
    ) {layerSetup [{branchFilters [{boneName=kin_sword_handle}]}]}) {layerGroup=None}
   n1 = Control Rig(Source=Copy Pose From Mesh() {bUseAttachedParent=true}) {controlRigAssetReference {controlRigAsset=&/M_MED_HeroicTale_Fair_CtrlRig}}
   Save cached pose 'controlRigCP'(Pose=@n1) {cacheName=controlRigCP}
   n2 = Anim Dynamics - Bone: dyn_chestPocket_inner_l!(ComponentPose=Local To Component(LocalPose=Use cached pose 'generalAdjustmentsCP'()), Alpha=0.5) -> (Pose) {
    angularDampingOverride=1
    angularSpringConstant=4000
    bAngularSpring=true
    bOverrideAngularDamping=true
    bUseGravityOverride=true
    gravityScale=0.35
    simulationSpace=RootRelative
    boundBone {boneName=dyn_chestPocket_inner_l}
    componentAppliedLinearAccClamp=<4000 4000 4000>
    componentLinearAccScale=<0.5 0.5>
    componentLinearVelScale=<0.5 0.5>
    gravityOverride {z=-980}
    physicsBodyDefinitions [{
      boundBone {boneName=dyn_chestPocket_inner_l}
      boxExtents=<10 10 10>
      constraintSetup {angularLimitsMax=<0 0 5>, angularLimitsMin=<0 -5 -5>, angularTarget=<1 0 0>}
      localJointOffset=<0.5 0 0>
     }]
   }
   n3 = Anim Dynamics - Bone: dyn_chestPocket_outer_l!(ComponentPose=@n2, Alpha=0.5) -> (Pose) {
    angularDampingOverride=1
    angularSpringConstant=4000
    bAngularSpring=true
    bOverrideAngularDamping=true
    bUseGravityOverride=true
    gravityScale=0.35
    simulationSpace=RootRelative
    boundBone {boneName=dyn_chestPocket_outer_l}
    componentAppliedLinearAccClamp=<4000 4000 4000>
    componentLinearAccScale=<0.5 0.5>
    componentLinearVelScale=<0.5 0.5>
    gravityOverride {z=-980}
    physicsBodyDefinitions [{
      boundBone {boneName=dyn_chestPocket_outer_l}
      boxExtents=<10 10 10>
      constraintSetup {angularLimitsMax=<0 0 5>, angularLimitsMin=<0 -5 -5>, angularTarget=<1 0 0>}
      localJointOffset=<0.5 0 0>
     }]
   }
   n4 = Anim Dynamics - Bone: dyn_chestPocket_inner_r!(ComponentPose=@n3, Alpha=0.5) -> (Pose) {
    angularDampingOverride=1
    angularSpringConstant=4000
    bAngularSpring=true
    bOverrideAngularDamping=true
    bUseGravityOverride=true
    gravityScale=0.35
    simulationSpace=RootRelative
    boundBone {boneName=dyn_chestPocket_inner_r}
    componentAppliedLinearAccClamp=<4000 4000 4000>
    componentLinearAccScale=<0.5 0.5>
    componentLinearVelScale=<0.5 0.5>
    gravityOverride {z=-980}
    physicsBodyDefinitions [{
      boundBone {boneName=dyn_chestPocket_inner_r}
      boxExtents=<10 10 10>
      constraintSetup {angularLimitsMax=<0 0 5>, angularLimitsMin=<0 -5 -5>, angularTarget=<1 0 0>}
      localJointOffset=<0.5 0 0>
     }]
   }
   n5 = Anim Dynamics - Bone: dyn_chestPocket_outer_r!(ComponentPose=@n4, Alpha=0.5) -> (Pose) {
    angularDampingOverride=1
    angularSpringConstant=4000
    bAngularSpring=true
    bOverrideAngularDamping=true
    bUseGravityOverride=true
    gravityScale=0.35
    simulationSpace=RootRelative
    boundBone {boneName=dyn_chestPocket_outer_r}
    componentAppliedLinearAccClamp=<4000 4000 4000>
    componentLinearAccScale=<0.5 0.5>
    componentLinearVelScale=<0.5 0.5>
    gravityOverride {z=-980}
    physicsBodyDefinitions [{
      boundBone {boneName=dyn_chestPocket_outer_r}
      boxExtents=<10 10 10>
      constraintSetup {angularLimitsMax=<0 0 5>, angularLimitsMin=<0 -5 -5>, angularTarget=<1 0 0>}
      localJointOffset=<0.5 0 0>
     }]
   }
   n6 = RigidBody!(ComponentPose=@n5, Alpha, OverrideWorldGravity:Vector="0,0,0") -> (Pose) {
    bClampLinearTranslationLimitToRefPose=true
    bOverrideWorldGravity=true
    overridePhysicsAsset=&/M_MED_HeroicTale_Fair_Physics
    simulationSpace=BaseBoneSpace
    alphaBoolBlend {blendOption=Linear}
    baseBoneRef {boneName=pelvis}
    componentAppliedLinearAccClamp=<3000 3000 3000>
    simSpaceSettings {
     maxAngularAcceleration=1
     maxAngularVelocity=3
     maxLinearAcceleration=600
     maxLinearVelocity=400
     velocityScaleZ=0.5
     worldAlpha=1
     externalLinearDragV=<1 1 1>
    }
   }
   n7 = RigidBody(ComponentPose=@n6, OverrideWorldGravity="0,0,250") {
    bClampLinearTranslationLimitToRefPose=true
    bOverrideWorldGravity=true
    overridePhysicsAsset=&/M_MED_HeroicTale_Fair_skirtAcc_Physics
    simulationSpace=BaseBoneSpace
    alphaBoolBlend {blendOption=Linear}
    baseBoneRef {boneName=spine_01}
    componentAppliedLinearAccClamp=<3000 3000 3000>
    simSpaceSettings {
     maxAngularAcceleration=2
     maxAngularVelocity=6
     maxLinearAcceleration=800
     maxLinearVelocity=400
     velocityScaleZ=0.5
     worldAlpha=1
     externalLinearDragV=<1 1 1>
    }
   }
   Save cached pose 'dynamicsCP'(Pose=Component To Local(ComponentPose=@n7)) {cacheName=dynamicsCP}
   n8 = "Constraint - attach_cape (P)"(ComponentPose=Local To Component(LocalPose=Use cached pose 'controlRigCP'()), Alpha=1, ConstraintWeights_0=1) {
    alphaBoolBlend {blendOption=Linear}
    boneToModify {boneName=attach_cape}
    constraintSetup [{targetBone {boneName=spine_05}}]
    constraintWeights [1]
   }
   n9 = "Transform (Modify) Bone - Bone: attach_cape"!(ComponentPose=@n8, bAlphaBoolEnabled:bool=True, Translation="0,-1,0", Rotation="0,0,16") -> (Pose) {
    alphaInputType=Bool
    rotationMode=BMM_Additive
    rotationSpace=BCS_BoneSpace
    translationMode=BMM_Additive
    translationSpace=BCS_BoneSpace
    alphaBoolBlend {blendInTime=0.5, blendOption=CircularOut, blendOutTime=0.5}
    boneToModify {boneName=attach_cape}
    rotation {roll=16}
    translation {y=-1}
   }
   n10 = "Transform (Modify) Bone - Bone: attach_cape"(ComponentPose=@n9, bAlphaBoolEnabled=True, Translation="0,-2,0", Rotation="0,0,20") {
    alphaInputType=Bool
    rotationMode=BMM_Additive
    rotationSpace=BCS_BoneSpace
    translationMode=BMM_Additive
    translationSpace=BCS_BoneSpace
    alphaBoolBlend {blendInTime=0.5, blendOption=CircularOut, blendOutTime=0.5}
    boneToModify {boneName=attach_cape}
    rotation {roll=20}
    translation {y=-2}
   }
   n11 = "Transform (Modify) Bone - Bone: attach_cape"(ComponentPose=@n10, bAlphaBoolEnabled=True, Translation="0,0,0", Rotation="0,0,10") {
    alphaInputType=Bool
    rotationMode=BMM_Additive
    rotationSpace=BCS_BoneSpace
    translationMode=BMM_Additive
    translationSpace=BCS_BoneSpace
    alphaBoolBlend {blendInTime=0.5, blendOption=Cubic, blendOutTime=0.5}
    boneToModify {boneName=attach_cape}
    rotation {roll=10}
   }
   n12 = "Transform (Modify) Bone - Bone: base_frontLoinCloth_1_l"!(ComponentPose=@n11, Alpha=1, Translation="0,0,2", Rotation="15,0,0") -> (Pose) {
    rotationMode=BMM_Additive
    rotationSpace=BCS_BoneSpace
    translationMode=BMM_Additive
    translationSpace=BCS_BoneSpace
    alphaBoolBlend {blendOption=Linear}
    alphaScaleBiasClamp {bInterpResult=true, interpSpeedDecreasing=0}
    boneToModify {boneName=base_frontLoinCloth_1_l}
    rotation {pitch=15}
    translation {z=2}
   }
   n13 = "Transform (Modify) Bone - Bone: base_frontLoinCloth_1_r"!(ComponentPose=@n12, Alpha=1, Translation="0,0,2", Rotation="15,0,0") -> (Pose) {
    rotationMode=BMM_Additive
    rotationSpace=BCS_BoneSpace
    translationMode=BMM_Additive
    translationSpace=BCS_BoneSpace
    alphaBoolBlend {blendOption=Linear}
    alphaScaleBiasClamp {bInterpResult=true, interpSpeedDecreasing=0}
    boneToModify {boneName=base_frontLoinCloth_1_r}
    rotation {pitch=15}
    translation {z=2}
   }
   n14 = "Transform (Modify) Bone - Bone: attach_backpack"!(ComponentPose=@n13, bAlphaBoolEnabled:bool=True, Translation="0,2.2,0", Rotation="0,0,0") -> (Pose) {
    alphaInputType=Bool
    rotationMode=BMM_Additive
    rotationSpace=BCS_BoneSpace
    translationMode=BMM_Additive
    translationSpace=BCS_BoneSpace
    alphaBoolBlend {blendOption=Linear}
    boneToModify {boneName=attach_backpack}
    translation {y=2.2}
   }
   n15 = "Transform (Modify) Bone - Bone: attach_cape"(ComponentPose=@n14, bAlphaBoolEnabled=True, Translation="0,2.2,0", Rotation="0,0,0") {
    alphaInputType=Bool
    rotationMode=BMM_Additive
    rotationSpace=BCS_BoneSpace
    translationMode=BMM_Additive
    translationSpace=BCS_BoneSpace
    alphaBoolBlend {blendOption=Linear}
    boneToModify {boneName=attach_cape}
    translation {y=2.2}
   }
   Save cached pose 'generalAdjustmentsCP'(Pose=Component To Local(ComponentPose=@n15)) {cacheName=generalAdjustmentsCP}
   n16 = "Transform (Modify) Bone - Bone: dyn_frontLoinCloth_1_l"!(
    ComponentPose=Local To Component(LocalPose=Use cached pose 'dynamicsCP'())
    bAlphaBoolEnabled:bool=True
    Translation="2.28,0,-2.78"
    Rotation="-0.43,-6.5,8.45"
   ) -> (Pose) {
    alphaInputType=Bool
    rotationMode=BMM_Additive
    rotationSpace=BCS_BoneSpace
    translationMode=BMM_Additive
    translationSpace=BCS_BoneSpace
    alphaBoolBlend {blendInTime=0.5, blendOption=Cubic, blendOutTime=0.5}
    boneToModify {boneName=dyn_frontLoinCloth_1_l}
    rotation=<rot -0.43 -6.5 8.45>
    translation {x=2.28, z=-2.78}
   }
   n17 = "Transform (Modify) Bone - Bone: dyn_frontLoinCloth_1_r"!(ComponentPose=@n16, bAlphaBoolEnabled:bool=True, Translation="0,0,0", Rotation="-9.2,0.4,-5.6") -> (Pose) {
    alphaInputType=Bool
    rotationMode=BMM_Additive
    rotationSpace=BCS_BoneSpace
    translationMode=BMM_Additive
    translationSpace=BCS_BoneSpace
    alphaBoolBlend {blendInTime=0.5, blendOption=Cubic, blendOutTime=0.5}
    boneToModify {boneName=dyn_frontLoinCloth_1_r}
    rotation=<rot -9.2 0.4 -5.6>
   }
   n18 = "Transform (Modify) Bone - Bone: dyn_frontLoinCloth_2_l"!(ComponentPose=@n17, bAlphaBoolEnabled:bool=True, Translation="0,0,0", Rotation="-10,-6,20") -> (Pose) {
    alphaInputType=Bool
    rotationMode=BMM_Additive
    rotationSpace=BCS_BoneSpace
    translationMode=BMM_Additive
    translationSpace=BCS_BoneSpace
    alphaBoolBlend {blendInTime=0.5, blendOption=Cubic, blendOutTime=0.5}
    boneToModify {boneName=dyn_frontLoinCloth_2_l}
    rotation=<rot -10 -6 20>
   }
   n19 = "Transform (Modify) Bone - Bone: kin_legbelt_l"!(ComponentPose=@n18, bAlphaBoolEnabled:bool=True, Translation="0.1,-0.2,-0.5", Rotation="0,-13,0") -> (Pose) {
    alphaInputType=Bool
    rotationMode=BMM_Additive
    rotationSpace=BCS_BoneSpace
    translationMode=BMM_Additive
    translationSpace=BCS_BoneSpace
    alphaBoolBlend {blendInTime=0.5, blendOption=Cubic, blendOutTime=0.5}
    boneToModify {boneName=kin_legbelt_l}
    rotation {yaw=-13}
    translation=<0.1 -0.2 -0.5>
   }
   n20 = "Transform (Modify) Bone - Bone: dyn_frontLoinCloth_1_l"(ComponentPose=@n19, bAlphaBoolEnabled=True, Translation="0,0,0", Rotation="0,10,0") {
    alphaInputType=Bool
    rotationMode=BMM_Additive
    rotationSpace=BCS_BoneSpace
    translationMode=BMM_Additive
    translationSpace=BCS_BoneSpace
    alphaBoolBlend {blendInTime=0.5, blendOption=Cubic, blendOutTime=0.5}
    alphaScaleBiasClamp {bInterpResult=true}
    boneToModify {boneName=dyn_frontLoinCloth_1_l}
    rotation {yaw=10}
   }
   n21 = "Transform (Modify) Bone - Bone: dyn_frontLoinCloth_2_r"!(ComponentPose=@n20, bAlphaBoolEnabled:bool=True, Translation="0,0,0", Rotation="0,-10,0") -> (Pose) {
    alphaInputType=Bool
    rotationMode=BMM_Additive
    rotationSpace=BCS_BoneSpace
    translationMode=BMM_Additive
    translationSpace=BCS_BoneSpace
    alphaBoolBlend {blendInTime=0.5, blendOption=Cubic, blendOutTime=0.5}
    alphaScaleBiasClamp {bInterpResult=true}
    boneToModify {boneName=dyn_frontLoinCloth_2_r}
    rotation {yaw=-10}
   }
   Save cached pose 'postSimFixes'(Pose=Component To Local(ComponentPose=@n21)) {cacheName=postSimFixes}
  ]}
 EventGraph {nodes [
   Event BlueprintUpdateAnimation()
   > n1 = Sequence!(execute) -> (then_0, then_1, then_2:Exec, then_3:Exec)
   > n2 = "Gravity Override State Function (Parameter Asset) (Nativized)"(
    CharacterPartAnimInstance=Self-Reference()
    GameTimeInSeconds=GetGameTimeInSeconds()
    Parameters=&/M_MED_HeroicTale_Fair_Parameters
   )
   > Set GravityOverride(GravityOverride=@n2)
   n3 = Get bIsTacticalSprinting()
   n4 = Get bIsFalling()
   n5 = Get bIsJumping()
   n6 = Branch(
    execute=@n1.then_1
    Condition=OR Boolean!(A=AND Boolean(A=@n5, B=@n3), B=AND Boolean(A=@n3, B=@n4), C:bool=Get bIsTacticalSprinting(), D:bool=Get bIsKneeSliding()) -> (ReturnValue)
   )
   > Set alphaLoinClothRB float(alphaLoinClothRB float:double=0.6) -> (Output_Get:double)
   Set alphaLoinClothRB float(execute=@n6.else, alphaLoinClothRB float=1)
   n7 = Get bIsMoving2D()
   Set isCrouchingStatic(execute=@n1.then_2, isCrouchingStatic:bool=AND Boolean(A=NOT Boolean(A=@n7), B=Get bIsCrouching())) -> (Output_Get:bool)
   n8 = Sequence!(execute=@n1.then_3) -> (then_0, then_1, then_2)
   n9 = Get bIsFalling()
   > n10 = Branch(Condition=AND Boolean(A=Get bIsTacticalSprinting(), B=NOT Boolean(A=@n9)))
   > Set moveLoinClothFw(moveLoinClothFw:double=1) -> (Output_Get:double)
   n11 = Branch(execute=@n10.else, Condition=Get bIsSprinting())
   > Set moveLoinClothFw(moveLoinClothFw=0.5)
   Set moveLoinClothFw(execute=@n11.else, moveLoinClothFw=0)
   n12 = PlayerMovement(execute=@n8.then_1, AnimInstance=Self-Reference(), XDotProductCoefficient=1, YDotProductCoefficient=1, ZDotProductCoefficient=1)
   n13 = AND Boolean(A=Get bIsCrouching(), B=Get bIsMoving2D())
   > Set moveLeftCrouching float(moveLeftCrouching float:bool=AND Boolean(A="float > float"(A=@n12, B=0.5), B=@n13)) -> (Output_Get:bool)
   > Set moveRightCrouching float(moveRightCrouching float:bool=AND Boolean(A=@n13, B="float < float"(A=@n12, B=-0.5))) -> (Output_Get:bool)
   Branch(execute=@n8.then_2, Condition=Get bIsInFrontEnd())
   > Set Custom Idle is Playing float(Custom Idle is Playing float:double=GetCurveValue(self=GetOwnerFrontEndAnimInstance(), CurveName=frontend_idle_bypass)) -> (Output_Get:double)
   Event OnParachuteOpened()
   > Switch on EGliderType(Selection=Get PlayerGliderType() -> (PlayerGliderType=HangGlider))
   > Set isUsingHangGlider(isUsingHangGlider:bool=true) -> (Output_Get:bool)
   Event OnParachuteClosed()
   > Set isUsingHangGlider(isUsingHangGlider=false)
   n14 = Event OnCharacterCustomizationLoaded()
   n15 = Get Array Item with EFort Custom Part Type(Array=@n14.CustomMeshParts, PartType=Face)
   > Set isWearingFAHood(isWearingFAHood:bool=MatchesWildcard(SourceString=GetDisplayName(Object=@n15), Wildcard=*M_MED_HeroicTale_Fair_Hood_FaceAcc, SearchCase=CaseSensitive)) -> (Output_Get:bool)
  ]}
]
variables [
 alphaLoinClothRB float {type=number}
 Custom Idle is Playing float {type=number}
 GravityOverride {type=Vector}
 isCrouchingStatic {type=boolean}
 isUsingHangGlider {type=boolean}
 isWearingFAHood {type=boolean}
 moveLeftCrouching float {type=boolean}
 moveLoinClothFw {type=number}
 moveRightCrouching float {type=boolean}
]
