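# Physics asset for the face accessory part of the male medium HeroicTale Fair character, set up for ponytail physics.
# Capsule bodies form a constraint chain from the ponytail base through two dynamic bones to the tip, with narrow swing limits and zero-degree twist limits.
# Unconstrained capsules on the upper spine and both sides of the lowered hood appear to act as colliders for the hair.
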PhysicsAsset /BRCosmetics/Characters/Player/Male/Medium/Bodies/M_MED_HeroicTale_Fair/Meshes/Parts/M_MED_HeroicTale_Fair_FaceAcc_Physics
bodies {
 base_ponytail_1 [{shapeType=Capsule, length=1, radius=0.5, shapeName=base_ponytail_1_capsule}]
 coll_hoodDownBack_l [{
   shapeType=Capsule
   length=12.2197
   radius=2.2803
   shapeName=coll_hoodDownBack_l_capsule
   center=<5.839 0.4753 -0.6226>
   rotation=<rot 90 0.8943 0>
  }]
 coll_hoodDownBack_r [{
   shapeType=Capsule
   length=10.9371
   radius=2.2803
   shapeName=coll_hoodDownBack_r_capsule
   center=<5.934 -0.3509 0.1278>
   rotation=<rot 90 0 0>
  }]
 dyn_ponytailTips_1 [{
   shapeType=Capsule
   length=1.1123
   radius=1.5882
   shapeName=dyn_ponytailTips_1_capsule
   center=<1.0915 -0.1378 -0.2262>
   rotation=<rot 0.5194 98.1337 -85.9872>
  }]
 dyn_ponytail_1 [{shapeType=Capsule, length=3.5172, radius=2.1645, shapeName=dyn_ponytail_1_capsule, center=<3.1082 -0.0417 0>, rotation=<rot 90 13.6327 0>}]
 dyn_ponytail_2 [{shapeType=Capsule, length=2.757, radius=2.5083, shapeName=dyn_ponytail_2_capsule, center=<2.6742 -0.0618 0>, rotation=<rot 90 -9.3053 0>}]
 spine_05 [{shapeType=Capsule, length=34.6636, radius=11.121, shapeName=spine_05_capsule, center=<6.5345 -5.8147 0>}]
}
constraints [
 {bone1Name=dyn_ponytail_2, bone2Name=dyn_ponytail_1, swing1LimitDegrees=30, swing2LimitDegrees=20, twistLimitDegrees=0}
 {bone1Name=dyn_ponytail_1, bone2Name=base_ponytail_1, swing1LimitDegrees=30, swing2LimitDegrees=20, twistLimitDegrees=0}
 {bone1Name=dyn_ponytailTips_1, bone2Name=dyn_ponytail_2, swing1LimitDegrees=20, swing2LimitDegrees=20, twistLimitDegrees=0}
]
